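# Physics asset for the back skirt of the medium female player Winter Holiday body, setting up ragdoll-style bodies for cloth-like skirt motion.
# Six two-bone chains of thin boxes (left and right back, back-middle and middle) hang from a pelvis capsule, alongside two thigh capsules.
# Skirt constraints keep twist and Swing1 tight while allowing wider Swing2, so panels flap mostly on one axis.
# Thigh-to-pelvis constraints leave all angular axes free.
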PhysicsAsset /Game/Characters/Player/Female/Medium/Bodies/F_MED_Winter_Holiday/Meshes/F_MED_Winter_Holiday_Skirt_Back_Physics_2
bodies {
 dyn_skirt_l_b_01 [{shapeType=Box, extentX=5.51, extentY=7.01, extentZ=1.01, center=<3.7414 0 0>}]
 dyn_skirt_l_b_02 [{shapeType=Box, extentX=6.01, extentY=5.51, extentZ=1.01, center=<4.2898 0 0>}]
 dyn_skirt_l_bm_01 [{shapeType=Box, extentX=5.51, extentY=7.01, extentZ=1.01, center=<3.7414 0 0>}]
 dyn_skirt_l_bm_02 [{shapeType=Box, extentX=6.01, extentY=5.51, extentZ=1.01, center=<4.2898 0 0>}]
 dyn_skirt_l_m_01 [{shapeType=Box, extentX=5.51, extentY=7.01, extentZ=1.01, center=<3.7414 0 0>}]
 dyn_skirt_l_m_02 [{shapeType=Box, extentX=6.01, extentY=5.51, extentZ=1.01, center=<4.2898 0 0>}]
 dyn_skirt_r_b_01 [{shapeType=Box, extentX=5.51, extentY=7.01, extentZ=1.01, center=<3.7414 0 0>}]
 dyn_skirt_r_b_02 [{shapeType=Box, extentX=6.01, extentY=5.51, extentZ=1.01, center=<4.2898 0 0>}]
 dyn_skirt_r_bm_01 [{shapeType=Box, extentX=5.51, extentY=7.01, extentZ=1.01, center=<3.7414 0 0>}]
 dyn_skirt_r_bm_02 [{shapeType=Box, extentX=6.01, extentY=5.51, extentZ=1.01, center=<4.2898 0 0>}]
 dyn_skirt_r_m_01 [{shapeType=Box, extentX=5.51, extentY=7.01, extentZ=1.01, center=<3.7414 0 0>}]
 dyn_skirt_r_m_02 [{shapeType=Box, extentX=6.01, extentY=5.51, extentZ=1.01, center=<4.2898 0 0>}]
 pelvis [{shapeType=Capsule, length=21.5349, radius=9.9482, shapeName=pelvis_capsule, rotation=<rot 90 0 0>}]
 thigh_l [{shapeType=Capsule, length=23.7267, radius=7.5, center=<-14.1887 0.3818 -0.011>, rotation=<rot -0.4847 -89.8283 -90.8209>}]
 thigh_r [{shapeType=Capsule, length=23.1152, radius=7.5, center=<14.345 -0.6916 -0.2676>, rotation=<rot -0.5856 -92.1317 -93.3232>}]
}
constraints [
 {
  bone1Name=thigh_l
  bone2Name=pelvis
  swing1LimitDegrees=45
  swing1Motion=Free
  swing2LimitDegrees=45
  swing2Motion=Free
  twistLimitDegrees=45
  twistMotion=Free
 }
 {bone1Name=dyn_skirt_r_m_01, bone2Name=pelvis, swing1LimitDegrees=2, swing2LimitDegrees=35, twistLimitDegrees=2}
 {bone1Name=dyn_skirt_r_m_02, bone2Name=dyn_skirt_r_m_01, swing1LimitDegrees=10, swing2LimitDegrees=35, twistLimitDegrees=2}
 {bone1Name=dyn_skirt_l_m_01, bone2Name=pelvis, swing1LimitDegrees=2, swing2LimitDegrees=35, twistLimitDegrees=2}
 {bone1Name=dyn_skirt_l_m_02, bone2Name=dyn_skirt_l_m_01, swing1LimitDegrees=10, swing2LimitDegrees=35, twistLimitDegrees=2}
 {bone1Name=dyn_skirt_l_b_01, bone2Name=pelvis, swing1LimitDegrees=2, swing2LimitDegrees=35, twistLimitDegrees=2}
 {bone1Name=dyn_skirt_l_b_02, bone2Name=dyn_skirt_l_b_01, swing1LimitDegrees=10, swing2LimitDegrees=35, twistLimitDegrees=2}
 {bone1Name=dyn_skirt_r_bm_01, bone2Name=pelvis, swing1LimitDegrees=2, swing2LimitDegrees=35, twistLimitDegrees=2}
 {bone1Name=dyn_skirt_r_bm_02, bone2Name=dyn_skirt_r_bm_01, swing1LimitDegrees=10, swing2LimitDegrees=35, twistLimitDegrees=2}
 {bone1Name=dyn_skirt_r_b_01, bone2Name=pelvis, swing1LimitDegrees=2, swing2LimitDegrees=35, twistLimitDegrees=2}
 {bone1Name=dyn_skirt_r_b_02, bone2Name=dyn_skirt_r_b_01, swing1LimitDegrees=10, swing2LimitDegrees=35, twistLimitDegrees=2}
 {bone1Name=dyn_skirt_l_bm_01, bone2Name=pelvis, swing1LimitDegrees=2, swing2LimitDegrees=35, twistLimitDegrees=2}
 {bone1Name=dyn_skirt_l_bm_02, bone2Name=dyn_skirt_l_bm_01, swing1LimitDegrees=10, swing2LimitDegrees=35, twistLimitDegrees=2}
 {
  bone1Name=thigh_r
  bone2Name=pelvis
  swing1LimitDegrees=45
  swing1Motion=Free
  swing2LimitDegrees=45
  swing2Motion=Free
  twistLimitDegrees=45
  twistMotion=Free
 }
]
