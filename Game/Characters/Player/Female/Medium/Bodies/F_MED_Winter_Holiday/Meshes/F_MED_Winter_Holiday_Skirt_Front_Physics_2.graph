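# Physics asset that simulates the front skirt of the medium female Winter Holiday player body.
# Capsules cover pelvis and thighs, tiny driver capsules anchor each side, and thin boxes form four two-bone panel chains.
# Panel joints allow moderate swing on one axis, tight swing on the other and almost no twist.
# Thigh and driver joints are free; the left driver joins thigh_l while the right joins pelvis.
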PhysicsAsset /Game/Characters/Player/Female/Medium/Bodies/F_MED_Winter_Holiday/Meshes/F_MED_Winter_Holiday_Skirt_Front_Physics_2
bodies {
 dyn_skirt_driver_l [{shapeType=Capsule, length=0.505, radius=0.505, rotation=<rot 0 -90 0>}]
 dyn_skirt_driver_r [{shapeType=Capsule, length=0.505, radius=0.505, rotation=<rot 0 -90 0>}]
 dyn_skirt_l_f_01 [{shapeType=Box, extentX=5.5, extentY=7.5, extentZ=1, center=<2.5 0 0>, rotation=<rot 0 0 90>}]
 dyn_skirt_l_f_02 [{shapeType=Box, extentX=8, extentY=8, extentZ=1, center=<4.5 0 0>, rotation=<rot 0 0 90>}]
 dyn_skirt_l_fm_01 [{shapeType=Box, extentX=5.5, extentY=7.5, extentZ=1, center=<3.25 0 0>, rotation=<rot 0 0 90>}]
 dyn_skirt_l_fm_02 [{shapeType=Box, extentX=8, extentY=8, extentZ=1, center=<5.25 0 0>, rotation=<rot 0 0 90>}]
 dyn_skirt_r_f_01 [{shapeType=Box, extentX=5.5, extentY=7.5, extentZ=1, center=<2.5 0 0>, rotation=<rot 0 0 90>}]
 dyn_skirt_r_f_02 [{shapeType=Box, extentX=8, extentY=8, extentZ=1, center=<4.5 0 0>, rotation=<rot 0 0 90>}]
 dyn_skirt_r_fm_01 [{shapeType=Box, extentX=5.5, extentY=7.5, extentZ=1, center=<3.25 0 0>, rotation=<rot 0 0 90>}]
 dyn_skirt_r_fm_02 [{shapeType=Box, extentX=8, extentY=8, extentZ=1, center=<5.25 0 0>, rotation=<rot 0 0 90>}]
 pelvis [{shapeType=Capsule, length=14.8992, radius=11.3533, center=<3.9116 -4.0729 -0.0058>, rotation=<rot 41.4655 -89.9999 89.7921>}]
 thigh_l [{shapeType=Capsule, length=29.6152, radius=8.25, center=<-14.16 -0.5949 -0.111>, rotation=<rot 0.6453 -90.5215 -92.4829>}]
 thigh_r [{shapeType=Capsule, length=29.6152, radius=8.25, center=<14.16 0.5949 0.111>, rotation=<rot 0.6453 -90.5215 -92.4829>}]
}
constraints [
 {
  bone1Name=thigh_l
  bone2Name=pelvis
  swing1LimitDegrees=45
  swing1Motion=Free
  swing2LimitDegrees=45
  swing2Motion=Free
  twistLimitDegrees=45
  twistMotion=Free
 }
 {
  bone1Name=thigh_r
  bone2Name=pelvis
  swing1LimitDegrees=45
  swing1Motion=Free
  swing2LimitDegrees=45
  swing2Motion=Free
  twistLimitDegrees=45
  twistMotion=Free
 }
 {
  bone1Name=dyn_skirt_driver_r
  bone2Name=pelvis
  swing1LimitDegrees=45
  swing1Motion=Free
  swing2LimitDegrees=45
  swing2Motion=Free
  twistLimitDegrees=45
  twistMotion=Free
 }
 {
  bone1Name=dyn_skirt_driver_l
  bone2Name=thigh_l
  swing1LimitDegrees=45
  swing1Motion=Free
  swing2LimitDegrees=45
  swing2Motion=Free
  twistLimitDegrees=45
  twistMotion=Free
 }
 {bone1Name=dyn_skirt_l_f_01, bone2Name=dyn_skirt_driver_l, swing1LimitDegrees=45, swing2LimitDegrees=10, twistLimitDegrees=2}
 {bone1Name=dyn_skirt_l_f_02, bone2Name=dyn_skirt_l_f_01, swing1LimitDegrees=45, swing2LimitDegrees=15, twistLimitDegrees=2}
 {bone1Name=dyn_skirt_l_fm_02, bone2Name=dyn_skirt_l_fm_01, swing1LimitDegrees=45, swing2LimitDegrees=10, twistLimitDegrees=2}
 {bone1Name=dyn_skirt_l_fm_01, bone2Name=dyn_skirt_driver_l, swing1LimitDegrees=25, swing2LimitDegrees=5, twistLimitDegrees=2}
 {bone1Name=dyn_skirt_r_f_01, bone2Name=dyn_skirt_driver_r, swing1LimitDegrees=45, swing2LimitDegrees=10, twistLimitDegrees=2}
 {bone1Name=dyn_skirt_r_f_02, bone2Name=dyn_skirt_r_f_01, swing1LimitDegrees=45, swing2LimitDegrees=15, twistLimitDegrees=2}
 {bone1Name=dyn_skirt_r_fm_01, bone2Name=dyn_skirt_driver_r, swing1LimitDegrees=25, swing2LimitDegrees=5, twistLimitDegrees=2}
 {bone1Name=dyn_skirt_r_fm_02, bone2Name=dyn_skirt_r_fm_01, swing1LimitDegrees=45, swing2LimitDegrees=10, twistLimitDegrees=2}
]
