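# Post-copy Anim Blueprint for the female medium Winter Holiday body, adding secondary motion to skirt, armband and collar.
# It copies the attached parent's pose, constrains skirt drivers to the thighs and offsets them, using alphas updated from crouch/sprint state and thigh yaw.
# Pelvis-space RigidBody skirt simulations feed AnimDynamics armband springs, then curve-based Bone Driven Controllers for the collar.
# An older bool-blended skirt offset branch remains unconnected to the output.
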AnimBlueprint /Game/Characters/Player/Female/Medium/Bodies/F_MED_Winter_Holiday/Meshes/F_MED_Winter_Holiday_Skeleton_AnimBlueprint
target_skeleton=&/F_MED_Winter_Holiday_Skeleton
graphs [
 AnimGraph {nodes [
   n1 = "Transform (Modify) Bone - Bone: dyn_skirt_driver_l"(
    ComponentPose=Local To Component(LocalPose=Use cached pose 'cachedpose'())
    Alpha=1
    Translation="-1.516,6.061,11.690"
    Rotation="0.000,0.001,-128.083"
    Scale="1,1,1"
   ) {
    rotationMode=BMM_Additive
    translationMode=BMM_Additive
    alphaBoolBlend {blendOption=Linear}
    boneToModify {boneName=dyn_skirt_driver_l}
    rotation {roll=-128.0828, yaw=0.001}
    translation=<-1.516 6.061 11.69>
   }
   n2 = "Transform (Modify) Bone - Bone: dyn_skirt_driver_r"(ComponentPose=@n1, Alpha=1, Translation="0.000,5.905,4.748", Rotation="0,0,0", Scale="1,1,1") {
    rotationMode=BMM_Additive
    translationMode=BMM_Additive
    alphaBoolBlend {blendOption=Linear}
    boneToModify {boneName=dyn_skirt_driver_r}
    translation {y=5.905, z=4.748}
   }
   Blend Poses by bool(BlendPose_0=Component To Local(ComponentPose=@n2), bActiveValue=Get bIsCrouchingOrSprinting() -> (bIsCrouchingOrSprinting:bool)) {blendType=Linear}
   n3 = Get Crouch / Sprint Alpha() -> (Crouch / Sprint Alpha:double)
   n4 = "Constraint - dyn_skirt_driver_l (P)"(ComponentPose=Local To Component(LocalPose=Use cached pose 'cachedpose'()), Alpha=@n3, ConstraintWeights_0=1) {
    alphaBoolBlend {blendOption=Linear}
    boneToModify {boneName=dyn_skirt_driver_l}
    constraintSetup [{targetBone {boneName=thigh_l}}]
    constraintWeights [1]
   }
   n5 = "Constraint - dyn_skirt_driver_r (P)"(ComponentPose=@n4, Alpha=@n3, ConstraintWeights_0=1) {
    alphaBoolBlend {blendOption=Linear}
    boneToModify {boneName=dyn_skirt_driver_r}
    constraintSetup [{targetBone {boneName=thigh_r}}]
    constraintWeights [1]
   }
   n6 = "Transform (Modify) Bone - Bone: dyn_skirt_driver_l"(
    ComponentPose=@n5
    Alpha=Get Crouch / Sprint L Alpha() -> (Crouch / Sprint L Alpha:double)
    Translation="0,1,-5"
    Rotation="0,0,0"
    Scale="1,1,1"
   ) {
    rotationMode=BMM_Additive
    rotationSpace=BCS_BoneSpace
    translationMode=BMM_Additive
    translationSpace=BCS_BoneSpace
    alphaBoolBlend {blendOption=Linear}
    boneToModify {boneName=dyn_skirt_driver_l}
    translation {y=1, z=-5}
   }
   n7 = "Transform (Modify) Bone - Bone: dyn_skirt_driver_r"(
    ComponentPose=@n6
    Alpha=Get Crouch / Sprint R Alpha() -> (Crouch / Sprint R Alpha:double)
    Translation="0,1,-5"
    Rotation="0,0,0"
    Scale="1,1,1"
   ) {
    rotationMode=BMM_Additive
    rotationSpace=BCS_BoneSpace
    translationMode=BMM_Additive
    translationSpace=BCS_BoneSpace
    alphaBoolBlend {blendOption=Linear}
    boneToModify {boneName=dyn_skirt_driver_r}
    translation {y=1, z=-5}
   }
   n8 = RigidBody(ComponentPose=@n7) {
    bOverrideWorldGravity=true
    overridePhysicsAsset=&/F_MED_Winter_Holiday_Skirt_Front_Physics_2
    simulationSpace=BaseBoneSpace
    alphaBoolBlend {blendOption=Linear}
    baseBoneRef {boneName=pelvis}
    componentAppliedLinearAccClamp=<3000 3000 3000>
    componentLinearAccScale=<0.3 0.3 0.5>
    componentLinearVelScale=<0.3 0.3 0.5>
    overrideWorldGravity {z=-980}
   }
   n9 = RigidBody(ComponentPose=@n8, Alpha=0.5) {
    bOverrideWorldGravity=true
    overridePhysicsAsset=&/F_MED_Winter_Holiday_Skirt_Back_Physics_2
    simulationSpace=BaseBoneSpace
    alphaBoolBlend {blendOption=Linear}
    baseBoneRef {boneName=pelvis}
    componentAppliedLinearAccClamp=<3000 3000 3000>
    componentLinearAccScale=<0.3 0.3 0.45>
    componentLinearVelScale=<0.3 0.3 0.45>
    overrideWorldGravity {z=-980}
   }
   n10 = Anim Dynamics - Bone: dyn_armband_knot(ComponentPose=@n9, Alpha=0.75, ExternalForce="0,0,0") {
    alpha=0.9
    angularDampingOverride=0.98
    angularSpringConstant=1500
    bAngularSpring=true
    bGravityOverrideInSimSpace=true
    bOverrideAngularDamping=true
    numSolverIterationsPostUpdate=4
    numSolverIterationsPreUpdate=16
    simulationSpace=Actor
    alphaBoolBlend {blendOption=Linear}
    boundBone {boneName=dyn_armband_knot}
    physicsBodyDefinitions [{
      sphereCollisionRadius=0
      boundBone {boneName=dyn_armband_knot}
      boxExtents=<15 5 5>
      constraintSetup {angularTargetAxis=AxisZ, twistAxis=AxisY, angularLimitsMax=<15 0 15>, angularLimitsMin=<-15 0 -15>, angularTarget=<0 0 1>}
      localJointOffset=<0 0 0>
     }]
   }
   n11 = Anim Dynamics - Bone: dyn_armband_low(ComponentPose=@n10, Alpha=0.5, ExternalForce="0,0,0") {
    alpha=0.9
    angularDampingOverride=0.98
    angularSpringConstant=1500
    bAngularSpring=true
    bGravityOverrideInSimSpace=true
    bOverrideAngularDamping=true
    numSolverIterationsPostUpdate=4
    numSolverIterationsPreUpdate=16
    simulationSpace=Actor
    alphaBoolBlend {blendOption=Linear}
    boundBone {boneName=dyn_armband_low}
    physicsBodyDefinitions [{
      sphereCollisionRadius=0
      boundBone {boneName=dyn_armband_low}
      boxExtents=<15 5 5>
      constraintSetup {angularTargetAxis=AxisZ, twistAxis=AxisY, angularLimitsMax=<15 15 15>, angularLimitsMin=<-15 -15 -15>, angularTarget=<0 0 1>}
      localJointOffset=<0 0 0>
     }]
   }
   n12 = Anim Dynamics - Bone: dyn_armband_hi(ComponentPose=@n11, Alpha=0.5, ExternalForce="0,0,0") {
    alpha=0.9
    angularDampingOverride=0.98
    angularSpringConstant=1500
    bAngularSpring=true
    bGravityOverrideInSimSpace=true
    bOverrideAngularDamping=true
    numSolverIterationsPostUpdate=4
    numSolverIterationsPreUpdate=16
    simulationSpace=Actor
    alphaBoolBlend {blendOption=Linear}
    boundBone {boneName=dyn_armband_hi}
    physicsBodyDefinitions [{
      sphereCollisionRadius=0
      boundBone {boneName=dyn_armband_hi}
      boxExtents=<15 5 5>
      constraintSetup {angularTargetAxis=AxisZ, twistAxis=AxisY, angularLimitsMax=<15 15 15>, angularLimitsMin=<-15 -15 -15>, angularTarget=<0 0 1>}
      localJointOffset=<0 0 0>
     }]
   }
   n13 = "dyn_colar_b_01.rotateZ = curve(neck_01.rotateZ) - Bone Driven Controller"!(ComponentPose:Component Space Pose Link Structure=@n12, Alpha:float=1) -> (Pose:Component Space Pose Link Structure) {
    is=title AnimGraphNode_BoneDrivenController
    bAffectTargetRotationZ=true
    drivingCurve=&/Bone_Driver_Winter_Holiday_Collar_back
    sourceComponent=RotationZ
    alphaBoolBlend {blendOption=Linear}
    sourceBone {boneName=neck_01}
    targetBone {boneName=dyn_colar_b_01}
   }
   n14 = "dyn_colar_b_01.multiple = curve(head.rotateZ) - Bone Driven Controller"!(ComponentPose:Component Space Pose Link Structure=@n13, Alpha:float=1) -> (Pose:Component Space Pose Link Structure) {
    is=title AnimGraphNode_BoneDrivenController
    bAffectTargetRotationX=true
    bAffectTargetRotationY=true
    drivingCurve=&/Bone_Driver_Winter_Holiday_ADS_Collar
    sourceComponent=RotationZ
    alphaBoolBlend {blendOption=Linear}
    sourceBone {boneName=head}
    targetBone {boneName=dyn_colar_b_01}
   }
   n15 = "dyn_colar_r_f_01.rotateY = curve(clavicle_r.rotateY) - Bone Driven Controller"!(ComponentPose:Component Space Pose Link Structure=@n14, Alpha:float=1) -> (Pose:Component Space Pose Link Structure) {
    is=title AnimGraphNode_BoneDrivenController
    bAffectTargetRotationY=true
    drivingCurve=&/Bone_Driver_Winter_Holiday_ADS_Collar_r
    multiplier=2
    sourceComponent=RotationY
    alphaBoolBlend {blendOption=Linear}
    sourceBone {boneName=clavicle_r}
    targetBone {boneName=dyn_colar_r_f_01}
   }
   n16 = "dyn_collar_l_f_01.multiple = curve(neck_01.rotateZ) - Bone Driven Controller"!(ComponentPose:Component Space Pose Link Structure=@n15, Alpha:float=0.5) -> (Pose:Component Space Pose Link Structure) {
    is=title AnimGraphNode_BoneDrivenController
    bAffectTargetRotationY=true
    bAffectTargetRotationZ=true
    drivingCurve=&/Bone_Driver_Winter_Holiday_ADS_Collar_l
    multiplier=2
    sourceComponent=RotationZ
    alphaBoolBlend {blendOption=Linear}
    sourceBone {boneName=neck_01}
    targetBone {boneName=dyn_collar_l_f_01}
   }
   Output Pose(Result=Component To Local(ComponentPose=@n16)) {layerGroup=None}
   Save cached pose 'cachedpose'(Pose=Copy Pose From Mesh() {bUseAttachedParent=true}) {cacheName=cachedpose}
  ]}
 EventGraph {nodes [
   Event BlueprintUpdateAnimation()
   n1 = Get bIsCrouching()
   n2 = Get bIsSprinting()
   > Set Crouch / Sprint Alpha(Crouch / Sprint Alpha:double=SelectFloat(A=1, B=0.65, bPickA=OR Boolean(A=@n1, B=@n2))) -> (Output_Get:double)
   n3 = BreakTransform(InTransform=GetDeltaTransformFromRefPose(self=GetOwnerSkeletalMeshComponent(), BoneName=thigh_l, BaseName=pelvis))
   > Set Crouch / Sprint L Alpha(Crouch / Sprint L Alpha:double=MapRangeClamped(Value=BreakRotator(InRot=@n3.Rotation).Yaw, InRangeA=-40, InRangeB=-100)) -> (Output_Get:double)
   n4 = BreakTransform(InTransform=GetDeltaTransformFromRefPose(self=GetOwnerSkeletalMeshComponent(), BoneName=thigh_r, BaseName=pelvis))
   > Set Crouch / Sprint R Alpha(Crouch / Sprint R Alpha:double=MapRangeClamped(Value=BreakRotator(InRot=@n4.Rotation).Yaw, InRangeA=-40, InRangeB=-100)) -> (Output_Get:double)
  ]}
]
variables [
 Crouch / Sprint Alpha {type=number}
 Crouch / Sprint L Alpha {type=number}
 Crouch / Sprint R Alpha {type=number}
 L Crouch Const Alpha {type=number}
 R Crouch Const Alpha {type=number}
]
